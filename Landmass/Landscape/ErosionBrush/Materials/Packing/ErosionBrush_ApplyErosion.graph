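# Unlit Landmass erosion brush material that applies an erosion texture to landscape height.
# It unpacks RG8 height from the LandscapeRT parameter, subtracts the Erosion red channel scaled by Saved Height Divisor and an optional mask, then repacks the result.
# The mask is a named reroute that blends from 1 to the Landmass shape mask's brush edge falloff using the UseMask scalar.
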Material /Landmass/Landscape/ErosionBrush/Materials/Packing/ErosionBrush_ApplyErosion
graph {
 properties {bAllowNegativeEmissiveColor=true, pixelDepthOffsetMode=PDOM_Legacy, shadingModel=MSM_Unlit, usedShadingModels=Unlit}
 nodes [
  n1 = TextureSampleParameter2D() {parameterName=LandscapeRT, texture=&Black}
  n2 = Multiply(
   A=TextureSampleParameter2D() {parameterName=Erosion, texture=/Engine/VREditor/Devices/Vive/UE_Logo}.R
   B=ScalarParameter() {defaultValue=512, parameterName=Saved Height Divisor}
  )
  n3 = NamedRerouteUsage()
  MaterialFunctionCall!(Height Int -32k to 32k=Subtract(
    A=MaterialFunctionCall!(RG=@n1.RGB) {materialFunction=/Landmass/Landscape/OldPrototype_BP/Mats/MF/UnpackRG8Height}.Result
    B=Multiply(A=@n2, B=@n3)
   )) {materialFunction=/Landmass/Landscape/OldPrototype_BP/Mats/MF/PackRG8Height}
  NamedRerouteDeclaration(None=LinearInterpolate(
    B=MaterialFunctionCall!(Falloff Width Override, Falloff Offset Override) {materialFunction=/Landmass/Landscape/BlueprintBrushes/Materials/BrushMaterials/BasicFunctions/LandmassBrush_ShapeMask}.Brush Edge Falloff
    Alpha=ScalarParameter() {parameterName=UseMask}
   ) {constA=1}) {name=Optional Mask, nodeColor=<rgba 0 1 0.1094 1>}
 ]
}
