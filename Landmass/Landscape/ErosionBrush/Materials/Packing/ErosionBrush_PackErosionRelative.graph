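# Unlit Landmass erosion brush material that, by its name, packs eroded height as values relative to the landscape.
# It unpacks the RG8 LandscapeRT height, subtracts HeightRT's red channel and divides by Saved Height Divisor, making values easier to see when inspecting the texture.
# A separate branch divides HeightRT by ZScale from MPC_Erosion to reach landscape scale.
# Negative emissive is allowed, so signed differences can be output.
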Material /Landmass/Landscape/ErosionBrush/Materials/Packing/ErosionBrush_PackErosionRelative
graph {
 properties {bAllowNegativeEmissiveColor=true, pixelDepthOffsetMode=PDOM_Legacy, shadingModel=MSM_Unlit, usedShadingModels=Unlit}
 nodes [
  n1 = TextureSampleParameter2D() {parameterName=LandscapeRT, texture=&Black}
  n2 = TextureSampleParameter2D() {parameterName=HeightRT, texture=&Black}
  Divide(
   A=Subtract(A=MaterialFunctionCall!(RG=@n1.RGB) {materialFunction=/Landmass/Landscape/OldPrototype_BP/Mats/MF/UnpackRG8Height}.Result, B=@n2.R) {desc=Make Relative}
   B=ScalarParameter() {defaultValue=512, parameterName=Saved Height Divisor}
  ) {desc="Divisor makes values more visible to the eye\r\nwhen inspecting the RT/Texture"}
  Divide(A=@n2.R, B=CollectionParameter() {collection=/Landmass/Landscape/BlueprintBrushes/MPC/MPC_Erosion, parameterName=ZScale}) {desc=Transform to Landscape Scale}
 ]
}
