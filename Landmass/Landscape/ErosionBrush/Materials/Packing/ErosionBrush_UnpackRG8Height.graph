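# Unlit utility material for the Landmass erosion brush that unpacks height data stored in an RG8-packed render target.
# The HeightRT texture parameter feeds UnpackRG8Height, whose result goes both into MakeFloat3 and into a multiply by ZScale from MPC_Erosion, labelled as a world-space conversion.
# Negative emissive color is allowed, though the graph shows no link from either branch to a material output.
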Material /Landmass/Landscape/ErosionBrush/Materials/Packing/ErosionBrush_UnpackRG8Height
graph {
 properties {bAllowNegativeEmissiveColor=true, pixelDepthOffsetMode=PDOM_Legacy, shadingModel=MSM_Unlit, usedShadingModels=Unlit}
 nodes [
  n1 = MaterialFunctionCall!(RG=TextureSampleParameter2D() {parameterName=HeightRT, texture=&Black}.RGB) {materialFunction=/Landmass/Landscape/OldPrototype_BP/Mats/MF/UnpackRG8Height}
  MaterialFunctionCall(X=@n1.Result) {materialFunction=&MakeFloat3}
  Multiply(A=@n1.Result, B=CollectionParameter() {collection=/Landmass/Landscape/BlueprintBrushes/MPC/MPC_Erosion, parameterName=ZScale}) {desc=To World Space}
 ]
}
